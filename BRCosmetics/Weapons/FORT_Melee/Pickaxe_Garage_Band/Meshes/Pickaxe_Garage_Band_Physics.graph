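# Physics asset for the Garage Band pickaxe that defines simulation bodies for its two-handed handle and a four-bone dangling cloth chain.
# A long capsule covers the handle, and small capsules follow the dyn_clothA bones.
# The first cloth bone is angularly locked to the handle as an anchor.
# Each later bone is constrained to the previous one, with wide swing angles that grow past the first joint and narrower twist angles.
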PhysicsAsset /BRCosmetics/Weapons/FORT_Melee/Pickaxe_Garage_Band/Meshes/Pickaxe_Garage_Band_Physics
bodies {
 Melee_TwoHanded_Handle [{shapeType=Capsule, length=80.8254, radius=1.747, center=<0 -1.6861 17.9949>, rotation=<rot 0 -90 0>}]
 dyn_clothA_01 [{shapeType=Capsule, length=0.505, radius=0.505, rotation=<rot 0 -90 0>}]
 dyn_clothA_02 [{shapeType=Capsule, length=3.2905, radius=1.6776, center=<0.6924 0.4162 -0.1251>, rotation=<rot 84.9531 -0.0001 -4.3125>}]
 dyn_clothA_03 [{shapeType=Capsule, length=3.5888, radius=1.618, center=<0.8767 0 -0.003>, rotation=<rot 89.7992 0.0008 0.0008>}]
 dyn_clothA_04 [{shapeType=Capsule, length=2.9016, radius=1.5, center=<0.7331 0 -0.0011>, rotation=<rot 89.9093 0.0186 0.0186>}]
}
constraints [
 {
  bone1Name=dyn_clothA_01
  bone2Name=Melee_TwoHanded_Handle
  swing1LimitDegrees=45
  swing1Motion=Locked
  swing2LimitDegrees=45
  swing2Motion=Locked
  twistLimitDegrees=45
  twistMotion=Locked
 }
 {bone1Name=dyn_clothA_02, bone2Name=dyn_clothA_01, swing1LimitDegrees=80, swing2LimitDegrees=80, twistLimitDegrees=26.8571}
 {bone1Name=dyn_clothA_03, bone2Name=dyn_clothA_02, swing1LimitDegrees=115, swing2LimitDegrees=115, twistLimitDegrees=20.7143}
 {bone1Name=dyn_clothA_04, bone2Name=dyn_clothA_03, swing1LimitDegrees=115, swing2LimitDegrees=115, twistLimitDegrees=32.7143}
]
