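# Animation Blueprint for the Garage Band pickaxe that adds secondary physics motion to its dangling cable and cloth bones.
# The input pose runs through Anim Dynamics on dyn_cable_top, dyn_clothB and dyn_cable_bot, then a RigidBody node using the pickaxe's physics asset.
# Several clothA Anim Dynamics nodes and a Trail controller are not connected to the output.
# The update event has no logic, and there are no variables.
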AnimBlueprint /BRCosmetics/Weapons/FORT_Melee/Pickaxe_Garage_Band/Meshes/Pickaxe_Garage_Band_Skeleton_AnimBlueprint
target_skeleton=&/Pickaxe_Garage_Band_Skeleton
graphs [
 AnimGraph {nodes [
   Anim Dynamics - Bone: dyn_clothA_02(Alpha=1, ExternalForce="0,0,0") {
    angularDampingOverride=1
    angularSpringConstant=2000
    bAngularSpring=true
    bGravityOverrideInSimSpace=true
    bOverrideAngularDamping=true
    gravityScale=0.85
    alphaBoolBlend {blendOption=Linear}
    boundBone {boneName=dyn_clothA_02}
    componentAppliedLinearAccClamp=<7000 7000 7000>
    componentLinearAccScale=<3 3 3>
    componentLinearVelScale=<6 6 6>
    physicsBodyDefinitions [{
      sphereCollisionRadius=0
      boundBone {boneName=dyn_clothA_02}
      boxExtents=<40 40 40>
      constraintSetup {twistAxis=AxisZ, angularLimitsMax=<0.01 20 45>, angularLimitsMin=<-0.01 -40 -45>, angularTarget=<0 0 0>}
      localJointOffset=<15 -7 0>
     }]
   }
   n1 = Anim Dynamics - Bone: dyn_clothA_02(Alpha=0.5, ExternalForce="0,0,0") {
    angularDampingOverride=1
    angularSpringConstant=700
    bAngularSpring=true
    bGravityOverrideInSimSpace=true
    bOverrideAngularDamping=true
    alphaBoolBlend {blendOption=Linear}
    boundBone {boneName=dyn_clothA_02}
    componentAppliedLinearAccClamp=<4000 4000 4000>
    componentLinearAccScale=<6 6 6>
    componentLinearVelScale=<3 3 3>
    physicsBodyDefinitions [{
      sphereCollisionRadius=0
      boundBone {boneName=dyn_clothA_02}
      boxExtents=<50 50 50>
      constraintSetup {twistAxis=AxisZ, angularLimitsMax=<10 20 0>, angularLimitsMin=<-90 -60 0>, angularTarget=<0 0 0>}
      localJointOffset=<0 0 -20>
     }]
   }
   Anim Dynamics - Bone: dyn_clothA_03(ComponentPose=@n1, Alpha=0.75, ExternalForce="0,0,0") {
    angularDampingOverride=1
    angularSpringConstant=500
    bAngularSpring=true
    bGravityOverrideInSimSpace=true
    bOverrideAngularDamping=true
    alphaBoolBlend {blendOption=Linear}
    boundBone {boneName=dyn_clothA_03}
    componentAppliedLinearAccClamp=<4000 4000 4000>
    componentLinearAccScale=<6 6 60>
    componentLinearVelScale=<3 3 3>
    physicsBodyDefinitions [{
      sphereCollisionRadius=0
      boundBone {boneName=dyn_clothA_03}
      boxExtents=<40 40 40>
      constraintSetup {twistAxis=AxisZ, angularLimitsMax=<15 20 0>, angularLimitsMin=<-60 -40 0>, angularTarget=<0 0 0>}
      localJointOffset=<0 0 -10>
     }]
   }
   Anim Dynamics - Bone: dyn_clothA_02(Alpha=1, ExternalForce="0,0,0") {
    angularDampingOverride=0.98
    angularSpringConstant=30000
    bGravityOverrideInSimSpace=true
    bOverrideAngularDamping=true
    alphaBoolBlend {blendOption=Linear}
    boundBone {boneName=dyn_clothA_02}
    componentAppliedLinearAccClamp=<10000 10000 10000>
    componentLinearAccScale=<5 5 5>
    componentLinearVelScale=<30 30 30>
    physicsBodyDefinitions [{
      sphereCollisionRadius=0
      boundBone {boneName=dyn_clothA_02}
      boxExtents=<50 50 50>
      constraintSetup {angularLimitsMax=<0 0 50>, angularLimitsMin=<0 0 -20>, angularTarget=<0 0 0>}
      localJointOffset=<0 -15 -30>
     }]
   }
   AnimGraphNode_SubInput_0 = Input Pose()
   n2 = Anim Dynamics - Bone: dyn_cable_top(ComponentPose=Local To Component(LocalPose=@AnimGraphNode_SubInput_0), Alpha=0.75, ExternalForce="0,0,0") {
    angularSpringConstant=30000
    bAngularSpring=true
    bGravityOverrideInSimSpace=true
    gravityScale=0.95
    lODThreshold=1
    alphaBoolBlend {blendOption=Linear}
    boundBone {boneName=dyn_cable_top}
    componentAppliedLinearAccClamp=<10000 10000 10000>
    componentLinearAccScale=<1.2 1.2 1.2>
    componentLinearVelScale=<1.1 1.1 1.1>
    physicsBodyDefinitions [{
      sphereCollisionRadius=0
      boundBone {boneName=dyn_cable_top}
      boxExtents=<20 75 80>
      constraintSetup {angularTargetAxis=AxisY, angularLimitsMax=<35 5 80>, angularLimitsMin=<-35 -5 -80>, angularTarget=<0 1 0>}
      localJointOffset=<0 -7 -15>
     }]
   }
   n3 = Anim Dynamics - Bone: dyn_clothB(ComponentPose=@n2, Alpha=0.55, ExternalForce="0,0,0") {
    angularSpringConstant=5000
    bAngularSpring=true
    bGravityOverrideInSimSpace=true
    alphaBoolBlend {blendOption=Linear}
    boundBone {boneName=dyn_clothB}
    componentAppliedLinearAccClamp=<5000 5000 5000>
    componentLinearAccScale=<1 1 1>
    componentLinearVelScale=<1 1 1>
    physicsBodyDefinitions [{
      sphereCollisionRadius=0
      boundBone {boneName=dyn_clothB}
      boxExtents=<24 40 6>
      constraintSetup {angularTargetAxis=AxisY, angularLimitsMax=<30 45 50>, angularLimitsMin=<-50 -20 -50>, angularTarget=<0 1 0>}
      localJointOffset=<0 -2 -4>
     }]
   }
   n4 = Anim Dynamics - Bone: dyn_cable_bot(ComponentPose=@n3, Alpha=0.9, ExternalForce="0,0,0") {
    angularSpringConstant=5000
    bAngularSpring=true
    bGravityOverrideInSimSpace=true
    alphaBoolBlend {blendOption=Linear}
    boundBone {boneName=dyn_cable_bot}
    componentAppliedLinearAccClamp=<5000 5000 5000>
    componentLinearAccScale=<1 1 1>
    componentLinearVelScale=<1 1 1>
    physicsBodyDefinitions [{
      sphereCollisionRadius=0
      boundBone {boneName=dyn_cable_bot}
      boxExtents=<18 30 6>
      constraintSetup {angularTargetAxis=AxisY, angularLimitsMax=<55 45 89.6>, angularLimitsMin=<-35 -20 -82.3999>, angularTarget=<0 1 0>}
      localJointOffset=<0 -2 -4>
     }]
   }
   n5 = RigidBody(ComponentPose=@n4) {
    overridePhysicsAsset=&/Pickaxe_Garage_Band_Physics
    alphaBoolBlend {blendOption=Linear}
    componentAppliedLinearAccClamp=<1000 1000 1000>
    componentLinearAccScale=<1 1 1>
    componentLinearVelScale=<8 8 8>
   }
   Output Pose(Result=Component To Local(ComponentPose=@n5)) {layerGroup=None}
   Trail controller - Bone: dyn_clothA_04(Alpha=1) {
    chainLength=3
    relaxationSpeedScale=2
    rotationLimits [{limitMax=<180 180 180>}, {limitMax=<180 180 180>}, {limitMax=<180 180 180>}]
    rotationOffsets [<0 0 0>, <0 0 0>, <0 0 0>]
    trailBone {boneName=dyn_clothA_04}
    trailRelaxationSpeed {editorCurveData {keys [<at 0 17 RCIM_Linear>,<at 1 10 RCIM_Linear>]}}
   }
  ]}
 EventGraph {nodes [Event BlueprintUpdateAnimation()]}
]
variables []
